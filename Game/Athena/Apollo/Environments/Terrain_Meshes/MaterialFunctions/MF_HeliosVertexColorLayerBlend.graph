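# Material function that blends Helios landscape layers on Apollo terrain meshes by vertex color, using layers from MF_Helios_StaticMeshLandscapeLayerMatch.
# Grass is the base, then gravel, sand, asphalt, scree, a custom input material and snow height-blend in order, each behind a static switch.
# Each blend weight is a selectable vertex channel plus an override; custom height travels in OpacityMask.
# Low and Medium quality replace the normal with the vertex normal before custom and snow.
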MaterialFunction /Game/Athena/Apollo/Environments/Terrain_Meshes/MaterialFunctions/MF_HeliosVertexColorLayerBlend
graph {nodes [
  n1 = StaticComponentMaskParameter(None=NamedRerouteUsage()) {
   defaultA=true
   desc=Select a vertex color to use as the weight for this layer.
   group=Helios_Layers
   parameterName=CustomMaterialLayer_VertexColor
   sortPriority=505
  }
  n2 = ScalarParameter() {
   desc=Set this to 1 if it's the only layer being used or is the bottom-most layer enabled.
   group=Helios_Layers
   parameterName=CustomMaterialLayer_OverrideVertexColor
   sortPriority=506
  }
  n3 = StaticSwitchParameter(True=NamedRerouteUsage(), False=GetMaterialAttributes().MaterialAttributes) {
   defaultValue=true
   desc="Grass does not use a vertex color, so it will appear where no vertex colors are painted. If Use_Grass is the only layer enabled, it will ignore vertex colors."
   group=Helios_Layers
   parameterName=Use_Grass
   sortPriority=100
  }
  n4 = StaticComponentMaskParameter(None=NamedRerouteUsage()) {
   defaultB=true
   desc=Select a vertex color to use as the weight for this layer.
   group=Helios_Layers
   parameterName=GravelLayer_VertexColor
   sortPriority=205
  }
  n5 = ScalarParameter() {
   desc=Set this to 1 if it's the only layer being used or is the bottom-most layer enabled.
   group=Helios_Layers
   parameterName=GravelLayer_OverrideVertexColor
   sortPriority=206
  }
  n6 = MaterialFunctionCall!(InputMaterial=@n3, LayerMaterial=NamedRerouteUsage(), LayerWeight=Saturate(None=Add(A=@n4, B=@n5)), UseContrast, Contrast) {materialFunction=/Game/Athena/Helios/Environments/Landscape/Materials/MaterialFunctions_Utilities/MF_HeightBlendLandscapeLayers_OpacityMask}
  n7 = StaticSwitchParameter(True=@n6.BlendedMaterial, False=@n3) {group=Helios_Layers, parameterName=Use_Gravel, sortPriority=200}
  n8 = StaticComponentMaskParameter(None=NamedRerouteUsage()) {
   defaultG=true
   desc=Select a vertex color to use as the weight for this layer.
   group=Helios_Layers
   parameterName=SandLayer_VertexColor
   sortPriority=305
  }
  n9 = ScalarParameter() {
   desc=Set this to 1 if it's the only layer being used or is the bottom-most layer enabled.
   group=Helios_Layers
   parameterName=SandLayer_OverrideVertexColor
   sortPriority=306
  }
  n10 = MaterialFunctionCall(InputMaterial=@n7, LayerMaterial=NamedRerouteUsage(), LayerWeight=Saturate(None=Add(A=@n8, B=@n9))) {materialFunction=&MF_HeightBlendLandscapeLayers_OpacityMask}
  n11 = StaticSwitchParameter(True=@n10.BlendedMaterial, False=@n7) {group=Helios_Layers, parameterName=Use_Sand, sortPriority=300}
  n12 = StaticComponentMaskParameter(None=NamedRerouteUsage()) {
   defaultR=true
   desc=Select a vertex color to use as the weight for this layer.
   group=Helios_Layers
   parameterName=AsphaltLayer_VertexColor
   sortPriority=405
  }
  n13 = ScalarParameter() {
   desc=Set this to 1 if it's the only layer being used or is the bottom-most layer enabled.
   group=Helios_Layers
   parameterName=AsphaltLayer_OverrideVertexColor
   sortPriority=406
  }
  n14 = MaterialFunctionCall(InputMaterial=@n11, LayerMaterial=NamedRerouteUsage(), LayerWeight=Saturate(None=Add(A=@n12, B=@n13))) {materialFunction=&MF_HeightBlendLandscapeLayers_OpacityMask}
  n15 = StaticSwitchParameter(True=@n14.BlendedMaterial, False=@n11) {group=Helios_Layers, parameterName=Use_Asphalt, sortPriority=400}
  n16 = StaticComponentMaskParameter(None=NamedRerouteUsage()) {
   defaultR=true
   desc=Select a vertex color to use as the weight for this layer.
   group=Helios_Layers
   parameterName=ScreeLayer_VertexColor
   sortPriority=405
  }
  n17 = ScalarParameter() {
   desc=Set this to 1 if it's the only layer being used or is the bottom-most layer enabled.
   group=Helios_Layers
   parameterName=ScreeLayer_OverrideVertexColor
   sortPriority=406
  }
  n18 = MaterialFunctionCall(InputMaterial=@n15, LayerMaterial=NamedRerouteUsage(), LayerWeight=Saturate(None=Add(A=@n16, B=@n17))) {materialFunction=&MF_HeightBlendLandscapeLayers_OpacityMask}
  n19 = StaticSwitchParameter(True=@n18.BlendedMaterial, False=@n15) {group=Helios_Layers, parameterName=Use_Scree, sortPriority=400}
  n20 = SetMaterialAttributes!(MaterialAttributes=@n19, Normal=VertexNormalWS()) {attributeSetTypes [#1]}
  n21 = QualitySwitch(Default=@n19, Low=@n20, Medium=@n20)
  n22 = MaterialFunctionCall(InputMaterial=@n21, LayerMaterial=NamedRerouteUsage(), LayerWeight=Saturate(None=Add(A=@n1, B=@n2))) {materialFunction=&MF_HeightBlendLandscapeLayers_OpacityMask}
  n23 = StaticSwitchParameter(True=@n22.BlendedMaterial, False=@n21) {group=Helios_Layers, parameterName=Use_CustomMaterialLayer, sortPriority=500}
  n24 = StaticComponentMaskParameter(None=NamedRerouteUsage()) {
   defaultG=true
   desc=Select a vertex color to use as the weight for this layer.
   group=Helios_Layers
   parameterName=SnowLayer_VertexColor
   sortPriority=605
  }
  n25 = ScalarParameter() {
   desc=Set this to 1 if it's the only layer being used or is the bottom-most layer enabled.
   group=Helios_Layers
   parameterName=SnowLayer_OverrideVertexColor
   sortPriority=606
  }
  n26 = MaterialFunctionCall(InputMaterial=@n23, LayerMaterial=NamedRerouteUsage(), LayerWeight=Saturate(None=Add(A=@n24, B=@n25))) {materialFunction=&MF_HeightBlendLandscapeLayers_OpacityMask}
  FunctionOutput(None=StaticSwitchParameter(True=@n26.BlendedMaterial, False=@n23) {group=Helios_Layers, parameterName=Use_Snow, sortPriority=600})
  n27 = MaterialFunctionCall!(Grass_UseCustomSlopeMask, Grass_CustomSlopeMask) {materialFunction=/Game/Athena/Helios/Environments/Landscape/Materials/MaterialFunctions_Utilities/MF_Helios_StaticMeshLandscapeLayerMatch}
  NamedRerouteDeclaration(None=@n27.GravelLayer) {name=GravelMaterial, nodeColor=<rgba 1 0 0.9844 1>}
  NamedRerouteDeclaration(None=@n27.AsphaltLayer) {name=AsphaltMaterial, nodeColor=<rgba 1 0 0.9844 1>}
  n28 = FunctionInput(Preview=GetMaterialAttributes().MaterialAttributes) {
   bUsePreviewValueAsDefault=true
   desc=The input normals need to be in worldspace
   inputName=CustomMaterial
   inputType=FunctionInput_MaterialAttributes
   sortPriority=16
  }
  n29 = Constant()
  NamedRerouteDeclaration(None=SetMaterialAttributes!(
    MaterialAttributes=@n28
    OpacityMask=FunctionInput(Preview=@n29) {bUsePreviewValueAsDefault=true, inputName=CustomMaterial_Height, inputType=FunctionInput_Scalar, sortPriority=17}
   ) {attributeSetTypes [#2]}) {name=CustomMaterial, nodeColor=<rgba 1 0 0.9844 1>}
  NamedRerouteDeclaration(None=@n27.SnowLayer) {name=SnowMaterial, nodeColor=<rgba 1 0 0.9844 1>}
  NamedRerouteDeclaration(None=@n27.GrassLayer) {name=GrassMaterial, nodeColor=<rgba 1 0 0.9844 1>}
  NamedRerouteDeclaration(None=@n27.ScreeLayer) {name=ScreeMaterial, nodeColor=<rgba 1 0 0.9844 1>}
  NamedRerouteDeclaration(None=@n27.SandLayer) {name=SandMaterial, nodeColor=<rgba 1 0 0.9844 1>}
  n30 = VertexColor()
  NamedRerouteDeclaration(None=AppendVector(A=@n30, B=@n30)) {name=VertexColor, nodeColor=<rgba 1 0 0.9844 1>}
 ]}
