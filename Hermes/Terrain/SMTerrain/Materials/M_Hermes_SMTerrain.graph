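# Master material for Hermes static-mesh terrain, supporting Nanite, displacement, spline and instanced meshes.
# Landscape-matched grass, gravel, mud, sand, asphalt, scree and custom layers height-blend through selectable vertex-color channels, with a cheaper single-layer switch and optional overlay.
# Optional runtime virtual texture blending uses height distance and green vertex color, with slope-based mip bias, skipping Low quality and HLOD proxies.
# Further sections provide Curie/grass hookup, masked paint and animated rice-water.
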Material /Hermes/Terrain/SMTerrain/Materials/M_Hermes_SMTerrain
graph {
 properties {
  bTangentSpaceNormal=false
  bUseMaterialAttributes=true
  bUsedWithInstancedStaticMeshes=true
  bUsedWithNanite=true
  bUsedWithSplineMeshes=true
  physMaterial=/Game/FortressPhysicalMaterials/Generic
  pixelDepthOffsetMode=PDOM_Legacy
  usedShadingModels=Default Lit
  displacementFadeRange {endSizePixels=3, startSizePixels=8}
  displacementScaling {center=0.5, magnitude=0.15}
 }
 nodes [
  Constant() {r=60}
  Constant() {r=6}
  Constant()
  n1 = LinearInterpolate(
   A=Constant3Vector() {constant=<rgba 0.0073 0.045 0.0027 1>}
   B=Constant3Vector() {constant=<rgba 0.135 0.0934 0.0614 1>}
   Alpha=VertexColor().B
  )
  n2 = Constant3Vector() {constant=<rgba 0.03 0.0192 0.0099 1>}
  n3 = VertexColor()
  MakeMaterialAttributes!(
   BaseColor=LinearInterpolate(A=LinearInterpolate(A=@n1, B=@n2, Alpha=@n3.A), B=Constant3Vector() {constant=<rgba 0.18 0.1795 0.1791 1>}, Alpha=VertexColor().R)
   Metallic=Constant()
   Specular
   Roughness=Constant() {r=1}
   Anisotropy
   EmissiveColor
   Opacity
   OpacityMask
   Normal=VertexNormalWS()
   Tangent
   WorldPositionOffset
   SubsurfaceColor
   ClearCoat
   ClearCoatRoughness
   AmbientOcclusion
   Refraction
   CustomizedUVs_0
   CustomizedUVs_1
   CustomizedUVs_2
   CustomizedUVs_3
   CustomizedUVs_4
   CustomizedUVs_5
   CustomizedUVs_6
   CustomizedUVs_7
   PixelDepthOffset
   ShadingModel
   Displacement
  )
  n4 = NamedRerouteUsage()
  n5 = DotProduct(A=VertexNormalWS(), B=Constant3Vector() {constant=<rgba 0 0 1 0>})
  n6 = RuntimeVirtualTextureSampleParameter!(
   UVs
   Translated World Position
   Mip Bias=LinearInterpolate(
    A=ScalarParameter() {defaultValue=6, group=RVT Base, parameterName=MipBiasVerticalSurfaces, sortPriority=2}
    B=ScalarParameter() {group=RVT Base, parameterName=MipBiasHorizontalSurfaces, sortPriority=3}
    Alpha=Saturate(None=@n5)
   )
  ) {
   group=RVT Base
   materialType=BaseColor_Normal_Specular_YCoCg
   mipValueMode=RVTMVM_MipBias
   parameterName=RVT Landscape Target
   virtualTexture=/Game/Athena/Apollo/Maps/Landscape/VT/RVT_Apollo_Landscape
   worldPositionOriginType=CameraRelative
  }
  n7 = SetMaterialAttributes!(MaterialAttributes=@n4, Normal=@n6.Normal) {attributeSetTypes [#1]}
  n8 = SetMaterialAttributes!(MaterialAttributes, Base Color=@n6.BaseColor, Specular=@n6.BaseColor, Roughness=@n6.BaseColor, Normal=@n6.Normal, Metallic=Constant()) {attributeSetTypes [
    #2
    #3
    #4
    #1
    #5
   ]}
  n9 = RuntimeVirtualTextureSampleParameter!(UVs, Translated World Position) {
   group=RVT Base
   materialType=WorldHeight
   parameterName=RVT Landscape Height Target
   virtualTexture=/Game/Athena/Apollo/Maps/Landscape/VT/RVT_Apollo_Landscape_Height
   worldPositionOriginType=CameraRelative
  }
  n10 = WorldPosition()
  n11 = Subtract(A=@n9.WorldHeight, B=@n10.Z)
  n12 = OneMinus(None=Ceil(None=Saturate(None=@n11)))
  n13 = ScalarParameter() {defaultValue=60, group=RVT Base, parameterName=BlendwithTerrainDistance, sliderMax=1000, sliderMin=0.001, sortPriority=1}
  n14 = Divide(A=Distance(A=@n9.WorldHeight, B=@n10.Z), B=Max(B=@n13) {constA=0.001})
  n15 = ScalarParameter() {defaultValue=1, desc=Currently uses Green channel, group=RVT Base, parameterName=UseVertColorToMaskOutRVT, sliderMax=1, sortPriority=1}
  n16 = Add(A=Saturate(None=Multiply(A=@n12, B=@n14)), B=Multiply(A=VertexColor().G, B=@n15))
  n17 = StaticSwitchParameter(
   True=BlendMaterialAttributes(A=StaticSwitchParameter(True=@n7, False=@n8) {group=RVT Base, parameterName=BlendOnlyNormalFromRVT}, B=@n4, Alpha=Saturate(None=@n16)) {vertexAttributeBlendType=UseA}
   False=@n4
  ) {
   defaultValue=true
   desc="Because of the expense, only use materials with UseRVT=True when the meshes are close enough to the landscape to warrant it! Otherwise, make sure to use materials with UseRVT=False!"
   group=RVT Base
   parameterName=UseRVT
   sortPriority=50
  }
  n18 = MaterialProxyReplace(
   Realtime=QualitySwitch(
    Default=MaterialFunctionCall(NoRVT=@n4, RVT=@n17) {materialFunction=&FN_RVT_Switch_Material}.Result
    Low=@n4
   )
   MaterialProxy=@n4
  ) {desc=Skip RVT on HLODs}
  n19 = StaticSwitchParameter(
   True=MaterialFunctionCall!(
    In=@n18
    UseHeightFalloff
    OnFireEmissiveOverride
    HeightFalloffRadius
    HeightFalloffHardness
    VariationMasksMobile
    In_Grass
    Is_Stylised_WE
   ) {materialFunction=/Game/Packages/Fortress_SharedMaterials/Base_Material_Functions/MF_Curie_Landscape}.Result
   False=@n18
  ) {
   defaultValue=true
   desc="Set to true to spawn grass on this mesh.\r\nUses the vertex color to determine where to spawn."
   group=GrassSpawner
   parameterName="Use Grass(true) or a BaseTilingTexture(False)"
   sortPriority=0
  }
  MaterialFunctionCall(MA In=@n19) {materialFunction=&MF_CreativeMoveTool}
  n20 = NamedRerouteUsage()
  n21 = SetMaterialAttributes!(
   MaterialAttributes
   Base Color=VectorParameter() {group=PaintLayer, parameterName=PaintLayerColor, defaultValue=<rgba 1 1 1 1>}.RGB
   Specular=Constant() {r=0.5}
   Roughness=ScalarParameter() {defaultValue=0.25, group=PaintLayer, parameterName=Paint Roughness}
   Normal=VertexNormalWS()
  ) {attributeSetTypes [
    #2
    #3
    #4
    #1
   ]}
  n22 = TextureCoordinate()
  n23 = VectorParameter() {group=PaintLayer, parameterName=PaintLayerUV_RG_Offset_BA_Scale, defaultValue=<rgba 0 0 1 1>}
  n24 = ComponentMask(None=@n23.RGB) {g=true, r=true}
  n25 = TextureSampleParameter2D(UVs=Multiply(A=Add(A=@n22, B=@n24), B=AppendVector(A=@n23.B, B=@n23.B))) {
   group=PaintLayer
   parameterName=PaintLayerMasks
   samplerType=SAMPLERTYPE_Masks
   texture=/Game/Packages/DS_Fortnight/SM/Textures/T_SoccerTennisBasketball
  }
  n26 = StaticComponentMaskParameter(None=AppendVector(A=@n25.RGB, B=@n25.RGB)) {defaultR=true, group=PaintLayer, parameterName=PaintLayerMaskChannel}
  n27 = WorldPosition()
  n28 = TextureSampleParameter2D(UVs=Divide(
    A=ComponentMask(None=@n27.XYZ) {g=true, r=true}
    B=ScalarParameter() {defaultValue=2048, group=PaintLayer, parameterName=PaintLayerMask_Mask Tiling}
   ) {constB=1024}) {
   group=PaintLayer
   parameterName=PaintLayerMasks_Mask
   samplerType=SAMPLERTYPE_Masks
   texture=/Game/Packages/Fortnite_UniqueMaterials_1/Textures/T_FRT_Conc_Masks_B_new
  }
  n29 = AppendVector(A=@n28.RGB, B=@n28.RGB)
  n30 = StaticSwitchParameter(
   True=Multiply(A=@n26, B=StaticComponentMaskParameter(None=@n29) {defaultR=true, group=PaintLayer, parameterName=PaintLayerMask_Mask_Channel})
   False=@n26
  ) {desc=Diffuse, group=PaintLayer, parameterName=Use Second PaintLayer Mask}
  n31 = Multiply(A=@n30, B=ScalarParameter() {defaultValue=1, group=PaintLayer, parameterName=PaintLayer Alpha Modulation})
  n32 = StaticSwitchParameter(
   True=BlendMaterialAttributes(A=@n20, B=@n21, Alpha=StaticSwitchParameter(True=@n31, False=@n30) {desc=Diffuse, group=PaintLayer, parameterName=Modulate Paint Layer Mask})
   False=@n20
  ) {desc=Uses a texture mask to paint a flat color on top of all other layers, group=Layers, parameterName=Use PaintLayer, sortPriority=100}
  NamedRerouteDeclaration(None=StaticSwitchParameter(True=NamedRerouteUsage(), False=NamedRerouteDeclaration(None=@n32) {name=Layers After Switch, nodeColor=<rgba 0 0.6953 1 1>}) {desc="R is the mask ", group=RiceWater, parameterName=RiceWater, sortPriority=-1}) {name=PaintLayer, nodeColor=<rgba 1 0 0.9844 1>}
  n33 = GetMaterialAttributes(None=NamedRerouteUsage()) {attributeGetTypes [#2]}
  n34 = ScalarParameter() {group=RiceWater, parameterName=WaterDepth Angle, sortPriority=3}
  n35 = Multiply(A=TextureCoordinate(), B=ScalarParameter() {defaultValue=5, group=RiceWater, parameterName=RiceWaterScale, sortPriority=2})
  n36 = ScalarParameter() {defaultValue=0.05, group=RiceWater, parameterName=RiceWaterSpeed}
  n37 = Panner(
   Coordinate=Multiply(A=@n35) {constB=2}
   Speed=ScalarParameter() {defaultValue=-0.05, group=RiceWater, parameterName=RiceWaterSpeed2, sortPriority=4}
  )
  n38 = Add(
   A=TextureSample(UVs=Panner(Coordinate=@n35, Speed=@n36)) {samplerType=SAMPLERTYPE_Normal, texture=/Game/Global/Textures/Default/Noise/T_SwirlNoise_N}.RGB
   B=Multiply(
    A=TextureSample(UVs=@n37) {samplerType=SAMPLERTYPE_Normal, texture=&T_SwirlNoise_N}.RGB
    B=Constant3Vector() {constant=<rgba 1 1 0 0>}
   )
  )
  n39 = LinearInterpolate(
   A=ScalarParameter() {defaultValue=1, group=RiceWater, parameterName=WaterDepth Face, sortPriority=3}
   B=@n34
   Alpha=ComponentMask(None=@n38) {r=true}
  )
  n40 = Fresnel() {baseReflectFraction=0, exponent=3}
  n41 = ScalarParameter() {defaultValue=0.01, group=RiceWater, parameterName=RiceWaterR, sortPriority=1}
  n42 = ScalarParameter() {defaultValue=1, group=RiceWater, parameterName=RiceWaterSpec, sortPriority=1}
  n43 = SetMaterialAttributes!(
   MaterialAttributes=@n33.MaterialAttributes
   Base Color=Multiply(A=@n33.Base Color, B=LinearInterpolate(A=@n39, B=@n34, Alpha=@n40)) {constB=0.25}
   Roughness=@n41
   Specular=@n42
   Normal=Multiply(A=@n38, B=Constant3Vector() {constant=<rgba 0.035 0.035 1 0>})
  ) {attributeSetTypes [
    #2
    #4
    #3
    #1
   ]}
  n44 = SetMaterialAttributes(
   MaterialAttributes=@n33.MaterialAttributes
   Base Color=Multiply(A=@n33.Base Color, B=@n34)
   Roughness=@n41
   Specular=@n42
   Normal=Constant3Vector() {constant=<rgba 0 0 1 0>}
  ) {attributeSetTypes [
    #2
    #4
    #3
    #1
   ]}
  NamedRerouteDeclaration(None=BlendMaterialAttributes(A=NamedRerouteUsage(), B=QualitySwitch(Default=@n43, Low=@n44, High=@n43, Medium=@n43, Epic=@n43), Alpha=VertexColor().R) {vertexAttributeBlendType=UseA}) {name=Rice, nodeColor=<rgba 1 0.5156 0 1>}
  n45 = VectorParameter() {group=LayerSetup, parameterName=AsphaltLayer_VertexColor, sortPriority=-699, defaultValue=<rgba 0 0 1 0>}
  n46 = AppendVector(A=@n45.RGB, B=@n45.RGB)
  n47 = NamedRerouteUsage()
  NamedRerouteDeclaration(None=Saturate(None=DotProduct(A=@n46, B=@n47))) {name=AsphaltLayer_Weight, nodeColor=<rgba 0 1 0 1>}
  n48 = VertexColor() {desc=mask cliff to grass transition}
  NamedRerouteDeclaration(None=MaterialFunctionCall(R=@n48.R, G=@n48.R, B=@n48.R, A=@n48.R) {materialFunction=&AppendMany}.RGBA) {name=VertexColor, nodeColor=<rgba 0 0 1 1>}
  n49 = MaterialFunctionCall() {materialFunction=/Game/Athena/Apollo/Environments/Terrain_Meshes/MaterialFunctions/MF_Common_GenericMaterial}
  n50 = MaterialFunctionCall!(CustomMaterial=@n49.Result, CustomMaterial_Height=@n49.Result) {materialFunction=/Game/Athena/Apollo/Environments/Terrain_Meshes/MaterialFunctions/MF_HeliosVertexColorLayerBlend}
  n51 = MaterialFunctionCall!(CustomMaterial=@n49.Result, CustomMaterial_Height=@n49.Result) {materialFunction=/Game/Athena/Apollo/Environments/Terrain_Meshes/MaterialFunctions/MF_AsteriaVextexColorLayerBlend}
  NamedRerouteDeclaration(None=StaticSwitchParameter(
    True=StaticSwitchParameter(True=@n49.Result, False=@n50.Result) {desc="If set to true, no other layers will be used.", group=Layers, parameterName=UseCustomMaterialOnly, sortPriority=-10}
    False=StaticSwitchParameter(True=@n49.Result, False=@n51.Result) {desc="If set to true, no other layers will be used.", group=Layers, parameterName=UseCustomMaterialOnly, sortPriority=-10}
   ) {group=Helios_Layers, parameterName=UseHeliosLayers, sortPriority=0}) {name=BlendedLayers, nodeColor=<rgba 1 0 0.9844 1>}
  n52 = VectorParameter() {group=LayerSetup, parameterName=ScreeLayer_VertexColor, sortPriority=-599}
  n53 = AppendVector(A=@n52.RGB, B=@n52.RGB)
  n54 = NamedRerouteUsage()
  NamedRerouteDeclaration(None=Saturate(None=DotProduct(A=@n53, B=@n54))) {name=ScreeLayer_Weight, nodeColor=<rgba 0 1 0 1>}
  n55 = NamedRerouteUsage()
  n56 = MaterialFunctionCall!(InputMaterial=@n55, LayerMaterial=NamedRerouteUsage(), LayerWeight=NamedRerouteUsage(), UseContrast, Contrast) {materialFunction=/Game/Athena/Helios/Environments/Landscape/Materials/MaterialFunctions_Utilities/MF_HeightBlendLandscapeLayers_OpacityMaskDisplaceCustom}
  n57 = StaticSwitchParameter(True=@n56.BlendedMaterial, False=@n55) {desc="Default=False", group=LayerSetup, parameterName=Use_GravelLayer, sortPriority=-900}
  n58 = MaterialFunctionCall(InputMaterial=@n57, LayerMaterial=NamedRerouteUsage(), LayerWeight=NamedRerouteUsage()) {materialFunction=&MF_HeightBlendLandscapeLayers_OpacityMaskDisplaceCustom}
  n59 = StaticSwitchParameter(True=@n58.BlendedMaterial, False=@n57) {desc="Default=False", group=LayerSetup, parameterName=Use_MudLayer, sortPriority=-890}
  n60 = MaterialFunctionCall(InputMaterial=@n59, LayerMaterial=NamedRerouteUsage(), LayerWeight=NamedRerouteUsage()) {materialFunction=&MF_HeightBlendLandscapeLayers_OpacityMaskDisplaceCustom}
  n61 = StaticSwitchParameter(True=@n60.BlendedMaterial, False=@n59) {desc="Default=False", group=LayerSetup, parameterName=Use_SandLayer, sortPriority=-800}
  n62 = MaterialFunctionCall(InputMaterial=@n61, LayerMaterial=NamedRerouteUsage(), LayerWeight=NamedRerouteUsage()) {materialFunction=&MF_HeightBlendLandscapeLayers_OpacityMaskDisplaceCustom}
  n63 = StaticSwitchParameter(True=@n62.BlendedMaterial, False=@n61) {desc="Default=False", group=LayerSetup, parameterName=Use_AsphaltLayer, sortPriority=-700}
  n64 = MaterialFunctionCall(InputMaterial=@n63, LayerMaterial=NamedRerouteUsage(), LayerWeight=NamedRerouteUsage()) {materialFunction=&MF_HeightBlendLandscapeLayers_OpacityMaskDisplaceCustom}
  n65 = StaticSwitchParameter(True=@n64.BlendedMaterial, False=@n63) {desc="Default=False", group=LayerSetup, parameterName=Use_ScreeLayer, sortPriority=-600}
  n66 = MaterialFunctionCall(InputMaterial=@n65, LayerMaterial=NamedRerouteUsage(), LayerWeight=NamedRerouteUsage()) {materialFunction=&MF_HeightBlendLandscapeLayers_OpacityMaskDisplaceCustom}
  NamedRerouteDeclaration(None=StaticSwitchParameter(True=@n66.BlendedMaterial, False=@n65) {desc="Default=False\r\nUses Vertex Alpha Channel", group=LayerSetup, parameterName=Use_CustomLayer, sortPriority=-500}) {name=MultipleLayerMaterial, nodeColor=<rgba 0 1 0 1>}
  n67 = NamedRerouteUsage()
  n68 = StaticSwitchParameter(
   True=NamedRerouteUsage()
   False=StaticSwitchParameter(True=NamedRerouteUsage(), False=NamedRerouteUsage() {desc=Grass should be at the end of the chain}) {desc="Default=False", group=LayerSetup, parameterName=Use_ScreeLayer, sortPriority=-600}
  ) {desc="Default=False", group=LayerSetup, parameterName=Use_AsphaltLayer, sortPriority=-700}
  NamedRerouteDeclaration(None=StaticSwitchParameter(
    True=NamedRerouteUsage()
    False=StaticSwitchParameter(True=@n67, False=@n68) {desc="Default=False", group=LayerSetup, parameterName=Use_SandLayer, sortPriority=-800}
   ) {desc="Default=False", group=LayerSetup, parameterName=Use_GravelLayer, sortPriority=-900}) {name=SingleLayerMaterial, nodeColor=<rgba 0 1 0 1>}
  n69 = StaticSwitchParameter(True=NamedRerouteUsage(), False=NamedRerouteUsage()) {
   defaultValue=true
   desc=Set this to true when only 1 layer is needed so the material is cheaper. Only Grass will show If all layers are set to false.
   group=LayerSetup
   parameterName=Use_SingleLayerONLY
   sortPriority=-1000
  }
  n70 = GetMaterialAttributes(None=@n69) {attributeGetTypes [#1, #2]}
  n71 = TextureSampleParameter2D() {
   group=Overlay
   parameterName=Overlay_Normal_Texture
   samplerSource=SSM_Wrap_WorldGroupSettings
   samplerType=SAMPLERTYPE_Normal
   sortPriority=-800
   texture=&FlatNormalBlue
  }
  n72 = Transform(None=MaterialFunctionCall(
    BaseNormal=Transform(None=@n70.Normal) {transformSourceType=TRANSFORMSOURCE_World, transformType=TRANSFORM_Tangent}
    AdditionalNormal=@n71.RGB
   ) {materialFunction=&BlendAngleCorrectedNormals}.Result)
  n73 = VectorParameter() {
   desc=This will use the Overlay_Mask_Texture as a mask to lerp between the incoming basecolor and this chosen color. Set Overlay_AO_Color_Mask_LerpAmount to Zero for no effect.
   group=Overlay
   parameterName=Overlay_AO_Color
   sortPriority=-600
   channelNames {a=Overlay_AO_Color_Mask_LerpIntensity, b="", g="", r=""}
  }
  n74 = TextureSampleParameter2D() {
   group=Overlay
   parameterName=Overlay_MASK_Texture
   samplerSource=SSM_Wrap_WorldGroupSettings
   samplerType=SAMPLERTYPE_Masks
   sortPriority=-700
   texture=&T_Fortnite_Default_S
  }
  n75 = OneMinus(None=@n74.R)
  n76 = LinearInterpolate(A=@n70.Base Color, B=@n73, Alpha=Saturate(None=Multiply(A=@n73.Overlay_AO_Color_Mask_LerpIntensity, B=@n75)))
  n77 = ScalarParameter() {
   desc=This will simply multiply the Overlay_MASK_Texture with the result after the Overlay_AO_Color.
   group=Overlay
   parameterName=Overlay_Multiply_MASK_ONLY_Amount
   sortPriority=-500
  }
  n78 = LinearInterpolate(A=@n76, B=Multiply(A=@n76, B=@n74.R), Alpha=Saturate(None=@n77))
  NamedRerouteDeclaration(None=StaticSwitch(
    True=SetMaterialAttributes!(MaterialAttributes=@n70.MaterialAttributes, Normal=@n72, Base Color=@n78) {attributeSetTypes [#1, #2]}
    False=@n69
    Value=StaticBoolParameter() {group=Overlay, parameterName=UseOverlay, sortPriority=-900}
   )) {name=MaterialBeforeRVT, nodeColor=<rgba 0 1 0 1>}
  n79 = MaterialFunctionCall!(Grass_UseCustomSlopeMask, Grass_CustomSlopeMask) {materialFunction=/BRCoreMaterials/Hermes/Terrain/Landscape/Materials/Layers/MF_Hermes_StaticMeshLandscapeLayerMatch}
  NamedRerouteDeclaration(None=@n79.MudLayer) {name=MudLayer, nodeColor=<rgba 0 1 0 1>}
  n80 = VectorParameter() {group=LayerSetup, parameterName=MudLayer_VertexColor, sortPriority=-789, defaultValue=<rgba 0 1 0 0>}
  n81 = AppendVector(A=@n80.RGB, B=@n80.RGB)
  n82 = NamedRerouteUsage()
  NamedRerouteDeclaration(None=Saturate(None=DotProduct(A=@n81, B=@n82))) {name=MudLayer_Weight, nodeColor=<rgba 0 1 0 1>}
  n83 = VectorParameter() {group=LayerSetup, parameterName=CustomLayer_VertexColor, sortPriority=-499}
  n84 = AppendVector(A=@n83.RGB, B=@n83.RGB)
  n85 = NamedRerouteUsage()
  NamedRerouteDeclaration(None=Saturate(None=DotProduct(A=@n84, B=@n85))) {name=CustomLayer_Weight, nodeColor=<rgba 0 1 0 1>}
  n86 = VectorParameter() {group=CustomLayer, parameterName=Custom_ColorTint, defaultValue=<rgba 1 1 1 0>}
  n87 = MaterialFunctionCall!(Scale=ScalarParameter() {defaultValue=900, group=CustomLayer, parameterName=Custom_WorldScale}) {materialFunction=/Game/Athena/Apollo/Environments/Landscape/Materials/Biome/CommonMaterials/MF_SimpleWorldUV}
  n88 = TextureSampleParameter2D(UVs=@n87.Result) {
   group=CustomLayer
   parameterName=Custom_D
   samplerSource=SSM_Wrap_WorldGroupSettings
   texture=/Game/Global/Textures/Default/Blanks/T_MidgreysRGB
  }
  n89 = TextureSampleParameter2D(UVs=@n87.Result) {
   group=CustomLayer
   parameterName=Custom_S
   samplerSource=SSM_Wrap_WorldGroupSettings
   samplerType=SAMPLERTYPE_Masks
   texture=&T_Fortnite_Default_S
  }
  n90 = TextureSampleParameter2D(UVs=@n87.Result) {
   group=CustomLayer
   parameterName=Custom_N
   samplerSource=SSM_Wrap_WorldGroupSettings
   samplerType=SAMPLERTYPE_Normal
   texture=&FlatNormalBlue
  }
  n91 = SetMaterialAttributes!(
   MaterialAttributes
   Base Color=Saturate(None=Multiply(A=@n86.RGB, B=@n88.RGB))
   Specular=@n89.R
   Roughness=@n89.R
   Normal=MaterialFunctionCall!(Normals=@n90.RGB, TransformToTangentSpace, Normal Strength) {materialFunction=/Game/Athena/Apollo/Environments/Landscape/Materials/Biome/CommonMaterials/MF_HQ_WorldAlignedNormal}.Result
   Opacity Mask=@n88.A
   Displacement=LinearInterpolate(B=@n88.A, Alpha=ScalarParameter() {defaultValue=1, group=Displacement, parameterName=Custom_DisplacementAmt}) {constA=0.5}
   Metallic=@n89.R
  ) {attributeSetTypes [
    #2
    #3
    #4
    #1
    #6
    #7
    #5
   ]}
  NamedRerouteDeclaration(None=@n91) {name=CustomLayer, nodeColor=<rgba 0 1 0 1>}
  n92 = Constant()
  n93 = SetMaterialAttributes!(MaterialAttributes, Base Color=@n92, Specular=@n92, Roughness=Constant() {r=1}, Metallic=@n92, Normal=VertexNormalWS()) {attributeSetTypes [
    #2
    #3
    #4
    #5
    #1
   ]}
  n94 = StaticSwitchParameter(True=@n79.GrassLayer, False=@n93) {
   defaultValue=true
   desc="Default=True. Grass will show up where no vertex color is present. Set this to false if no grass is needed."
   group=LayerSetup
   parameterName=Use_GrassLayerAsDefault
   sortPriority=-1000
  }
  NamedRerouteDeclaration(None=@n94) {name=GrassLayer, nodeColor=<rgba 0 1 0 1>}
  NamedRerouteDeclaration(None=@n79.GravelLayer) {name=GravelLayer, nodeColor=<rgba 0 1 0 1>}
  NamedRerouteDeclaration(None=@n79.SandLayer) {name=SandLayer, nodeColor=<rgba 0 1 0 1>}
  NamedRerouteDeclaration(None=@n79.AsphaltLayer) {name=AsphaltLayer, nodeColor=<rgba 0 1 0 1>}
  NamedRerouteDeclaration(None=@n79.ScreeLayer) {name=ScreeLayer, nodeColor=<rgba 0 1 0 1>}
  n95 = VectorParameter() {group=LayerSetup, parameterName=GravelLayer_VertexColor, sortPriority=-899, defaultValue=<rgba 1 0 0 0>}
  n96 = AppendVector(A=@n95.RGB, B=@n95.RGB)
  n97 = NamedRerouteUsage()
  NamedRerouteDeclaration(None=Saturate(None=DotProduct(A=@n96, B=@n97))) {name=GravelLayer_Weight, nodeColor=<rgba 0 1 0 1>}
  n98 = VectorParameter() {group=LayerSetup, parameterName=SandLayer_VertexColor, sortPriority=-799, defaultValue=<rgba 0 1 0 0>}
  n99 = AppendVector(A=@n98.RGB, B=@n98.RGB)
  n100 = NamedRerouteUsage()
  NamedRerouteDeclaration(None=Saturate(None=DotProduct(A=@n99, B=@n100))) {name=SandLayer_Weight, nodeColor=<rgba 0 1 0 1>}
  n101 = GetMaterialAttributes(None=@n18) {attributeGetTypes [
    #2
    #3
    #4
    #1
    #7
   ]}
  n102 = SetMaterialAttributes!(MaterialAttributes, Base Color=@n101.Base Color, Specular=@n101.Base Color, Roughness=@n101.Base Color, Normal=@n101.Base Color) {attributeSetTypes [
    #2
    #3
    #4
    #1
   ]}
  n103 = SetMaterialAttributes!(
   MaterialAttributes
   Base Color=@n101.Base Color
   Specular=@n101.Base Color
   Roughness=@n101.Base Color
   Normal=@n101.Base Color
   Displacement=@n101.Base Color
  ) {attributeSetTypes [
    #2
    #3
    #4
    #1
    #7
   ]}
  QualitySwitch(Default=@n102, Epic=@n103)
  n104 = RuntimeVirtualTextureSampleParameter() {
   group=RVT Base
   materialType=Displacement
   parameterName=RVT Displacement
   virtualTexture=/Game/Athena/Apollo/Maps/Landscape/VT/RVT_Helios_Displacement
   worldPositionOriginType=CameraRelative
  }
  n105 = SetMaterialAttributes(Base Color=@n6.BaseColor, Specular=@n6.BaseColor, Roughness=@n6.BaseColor, Normal=@n6.Normal, Displacement=@n104.Displacement) {attributeSetTypes [
    #2
    #3
    #4
    #1
    #7
   ]}
  QualitySwitch(Default=@n8, Epic=@n105)
  Reroute(None=@n32)
  ScalarParameter() {group=PaintLayer, parameterName=Paint Metallic}
  n106 = GetMaterialAttributes(None=@n18) {attributeGetTypes [
    #2
    #3
    #4
    #1
    #8
    #9
    #10
   ]}
  SetMaterialAttributes!(
   MaterialAttributes
   Base Color=@n106.Base Color
   Specular=@n106.Base Color
   Roughness=@n106.Base Color
   Normal=@n106.Base Color
   Subsurface Color=@n106.Base Color
   Opacity=@n106.Base Color
   Shading Model=@n106.Base Color
  ) {attributeSetTypes [
    #2
    #3
    #4
    #1
    #8
    #9
    #10
   ]}
  SetMaterialAttributes!(MaterialAttributes=@n18, Shading Model=ShadingModel()) {attributeSetTypes [#10]}
  StaticSwitchParameter() {parameterName=SpawnPCG_Grass}
  n107 = StaticSwitchParameter(True=Constant() {r=1}, False=Constant()) {
   defaultValue=true
   desc="Set to true to spawn grass on this mesh.\r\nUses the vertex color to determine where to spawn."
   group=GrassSpawner
   parameterName="Use Grass(true) or a BaseTilingTexture(False)"
   sortPriority=0
  }
  StaticSwitchParameter(True=@n107) {defaultValue=true, group=Layers, parameterName=Use a base Layer}
 ]
}
